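# Physics asset for the BrakePedal backpack cosmetic, giving its pack, chain head and six leaf bones bodies and joints for physics motion.
# Two capsules shape the pack, and single capsules cover each base and dyn bone.
# Each dyn bone joins its matching base bone, which in turn joins the pack.
# Only the dyn-to-base chain head joint restricts swing and twist; all other joints leave angular motion free.
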PhysicsAsset /BRCosmetics/Accessories/FORT_Backpacks/Backpack_M_MED_BrakePedal/Meshes/M_MED_BrakePedal_Pack_Physics
bodies {
 base_chain_head [{shapeType=Capsule, length=1, radius=1, shapeName=base_chain_head_capsule, rotation=<rot -90 0 0>}]
 base_leaf_1 [{shapeType=Capsule, length=1, radius=1, shapeName=base_leaf_1_capsule, rotation=<rot -90 0 0>}]
 base_leaf_2 [{shapeType=Capsule, length=1, radius=1, shapeName=base_leaf_2_capsule, rotation=<rot -90 0 0>}]
 base_leaf_3 [{shapeType=Capsule, length=1, radius=1, shapeName=base_leaf_3_capsule, rotation=<rot -90 0 0>}]
 base_leaf_4 [{shapeType=Capsule, length=1, radius=1, shapeName=base_leaf_4_capsule, rotation=<rot -90 0 0>}]
 base_leaf_5 [{shapeType=Capsule, length=1, radius=1, shapeName=base_leaf_5_capsule, rotation=<rot -90 0 0>}]
 base_leaf_6 [{shapeType=Capsule, length=1, radius=1, shapeName=base_leaf_6_capsule, rotation=<rot -90 0 0>}]
 dyn_chain_head [{shapeType=Capsule, length=0.5, radius=3.25, shapeName=dyn_chain_head_capsule, center=<5 0 -0.28>, rotation=<rot -90 0 0>}]
 dyn_leaf_1 [{shapeType=Capsule, length=1, radius=1, shapeName=dyn_leaf_1_capsule, center=<2 0 0>, rotation=<rot -90 0 0>}]
 dyn_leaf_2 [{shapeType=Capsule, length=1, radius=1, shapeName=dyn_leaf_2_capsule, center=<2 0 0>, rotation=<rot -90 0 0>}]
 dyn_leaf_3 [{shapeType=Capsule, length=1, radius=1, shapeName=dyn_leaf_3_capsule, center=<2 0 0>, rotation=<rot -90 0 0>}]
 dyn_leaf_4 [{shapeType=Capsule, length=1, radius=1, shapeName=dyn_leaf_4_capsule, center=<2 0 0>, rotation=<rot -90 0 0>}]
 dyn_leaf_5 [{shapeType=Capsule, length=1, radius=1, shapeName=dyn_leaf_5_capsule, center=<2 0 0>, rotation=<rot -90 0 0>}]
 dyn_leaf_6 [{shapeType=Capsule, length=1, radius=1, shapeName=dyn_leaf_6_capsule, center=<2 0 0>, rotation=<rot -90 0 0>}]
 pack [
  {shapeType=Capsule, length=5, radius=3, shapeName=pack_capsule, center=<-6.478 2.7091 -15.5778>, rotation=<rot -10 0.0001 -0.0005>}
  {shapeType=Capsule, length=8, radius=4.5, center=<-1.5542 -4.4663 -9.4715>, rotation=<rot 80 0 0>}
 ]
}
constraints [
 {bone1Name=dyn_chain_head, bone2Name=base_chain_head, swing1LimitDegrees=50, swing2LimitDegrees=50, twistLimitDegrees=15}
 {
  bone1Name=dyn_leaf_1
  bone2Name=base_leaf_1
  swing1LimitDegrees=45
  swing1Motion=Free
  swing2LimitDegrees=45
  swing2Motion=Free
  twistLimitDegrees=45
  twistMotion=Free
 }
 {
  bone1Name=dyn_leaf_2
  bone2Name=base_leaf_2
  swing1LimitDegrees=45
  swing1Motion=Free
  swing2LimitDegrees=45
  swing2Motion=Free
  twistLimitDegrees=45
  twistMotion=Free
 }
 {
  bone1Name=dyn_leaf_3
  bone2Name=base_leaf_3
  swing1LimitDegrees=45
  swing1Motion=Free
  swing2LimitDegrees=45
  swing2Motion=Free
  twistLimitDegrees=45
  twistMotion=Free
 }
 {
  bone1Name=dyn_leaf_4
  bone2Name=base_leaf_4
  swing1LimitDegrees=45
  swing1Motion=Free
  swing2LimitDegrees=45
  swing2Motion=Free
  twistLimitDegrees=45
  twistMotion=Free
 }
 {
  bone1Name=dyn_leaf_5
  bone2Name=base_leaf_5
  swing1LimitDegrees=45
  swing1Motion=Free
  swing2LimitDegrees=45
  swing2Motion=Free
  twistLimitDegrees=45
  twistMotion=Free
 }
 {
  bone1Name=dyn_leaf_6
  bone2Name=base_leaf_6
  swing1LimitDegrees=45
  swing1Motion=Free
  swing2LimitDegrees=45
  swing2Motion=Free
  twistLimitDegrees=45
  twistMotion=Free
 }
 {
  bone1Name=base_leaf_1
  bone2Name=pack
  swing1LimitDegrees=45
  swing1Motion=Free
  swing2LimitDegrees=45
  swing2Motion=Free
  twistLimitDegrees=45
  twistMotion=Free
 }
 {
  bone1Name=base_leaf_2
  bone2Name=pack
  swing1LimitDegrees=45
  swing1Motion=Free
  swing2LimitDegrees=45
  swing2Motion=Free
  twistLimitDegrees=45
  twistMotion=Free
 }
 {
  bone1Name=base_leaf_3
  bone2Name=pack
  swing1LimitDegrees=45
  swing1Motion=Free
  swing2LimitDegrees=45
  swing2Motion=Free
  twistLimitDegrees=45
  twistMotion=Free
 }
 {
  bone1Name=base_leaf_4
  bone2Name=pack
  swing1LimitDegrees=45
  swing1Motion=Free
  swing2LimitDegrees=45
  swing2Motion=Free
  twistLimitDegrees=45
  twistMotion=Free
 }
 {
  bone1Name=base_leaf_5
  bone2Name=pack
  swing1LimitDegrees=45
  swing1Motion=Free
  swing2LimitDegrees=45
  swing2Motion=Free
  twistLimitDegrees=45
  twistMotion=Free
 }
 {
  bone1Name=base_leaf_6
  bone2Name=pack
  swing1LimitDegrees=45
  swing1Motion=Free
  swing2LimitDegrees=45
  swing2Motion=Free
  twistLimitDegrees=45
  twistMotion=Free
 }
 {
  bone1Name=base_chain_head
  bone2Name=pack
  swing1LimitDegrees=45
  swing1Motion=Free
  swing2LimitDegrees=45
  swing2Motion=Free
  twistLimitDegrees=45
  twistMotion=Free
 }
]
