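# Anim Blueprint for the BrakePedal backpack, adding procedural secondary motion and keeping the pack posed against the wearer's body.
# Input pose gets a fixed pack offset plus spine-bend-driven roll and shift, then feeds six chained AnimDynamics leaf bones and a RigidBody simulation.
# A final pack roll follows left upper-arm rotation, weighted by smoothed melee state times head-down alpha.
# EventGraph computes these values and RigidBody gravity via a deprecated helper.
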AnimBlueprint /BRCosmetics/Accessories/FORT_Backpacks/Backpack_M_MED_BrakePedal/Meshes/M_MED_BrakePedal_Pack_AnimBP
target_skeleton=/Game/Accessories/FORT_Backpacks/Backpack_M_MED_BrakePedal/Meshes/M_MED_BrakePedal_Pack_Skeleton
graphs [
 AnimGraph {nodes [
   n1 = Anim Dynamics - Bone: dyn_leaf_1(ComponentPose=Local To Component(LocalPose=Use cached pose 'start_Cache'()), Alpha=1, ExternalForce="0,0,0") {
    angularDampingOverride=1
    angularSpringConstant=20000
    bAngularSpring=true
    bOverrideAngularDamping=true
    bUseGravityOverride=true
    gravityScale=0.35
    simulationSpace=RootRelative
    alphaBoolBlend {blendOption=Linear}
    boundBone {boneName=dyn_leaf_1}
    componentLinearAccScale=<1 1 1>
    componentLinearVelScale=<1 1 1>
    gravityOverride {z=-980}
    physicsBodyDefinitions [{
      boundBone {boneName=dyn_leaf_1}
      boxExtents=<20 20 20>
      constraintSetup {angularLimitsMax=<10 45 45>, angularLimitsMin=<-10 -45 -45>, angularTarget=<1 0 0>}
      localJointOffset=<3 0 0>
     }]
   }
   n2 = Anim Dynamics - Bone: dyn_leaf_2(ComponentPose=@n1, Alpha=1, ExternalForce="0,0,0") {
    angularDampingOverride=1
    angularSpringConstant=20000
    bAngularSpring=true
    bOverrideAngularDamping=true
    bUseGravityOverride=true
    gravityScale=0.35
    simulationSpace=RootRelative
    alphaBoolBlend {blendOption=Linear}
    boundBone {boneName=dyn_leaf_2}
    componentLinearAccScale=<1 1 1>
    componentLinearVelScale=<1 1 1>
    gravityOverride {z=-980}
    physicsBodyDefinitions [{
      boundBone {boneName=dyn_leaf_2}
      boxExtents=<20 20 20>
      constraintSetup {angularLimitsMax=<10 45 45>, angularLimitsMin=<-10 -45 -45>, angularTarget=<1 0 0>}
      localJointOffset=<3 0 0>
     }]
   }
   n3 = Anim Dynamics - Bone: dyn_leaf_3(ComponentPose=@n2, Alpha=1, ExternalForce="0,0,0") {
    angularDampingOverride=1
    angularSpringConstant=20000
    bAngularSpring=true
    bOverrideAngularDamping=true
    bUseGravityOverride=true
    gravityScale=0.35
    simulationSpace=RootRelative
    alphaBoolBlend {blendOption=Linear}
    boundBone {boneName=dyn_leaf_3}
    componentLinearAccScale=<1 1 1>
    componentLinearVelScale=<1 1 1>
    gravityOverride {z=-980}
    physicsBodyDefinitions [{
      boundBone {boneName=dyn_leaf_3}
      boxExtents=<20 20 20>
      constraintSetup {angularLimitsMax=<10 45 45>, angularLimitsMin=<-10 -45 -45>, angularTarget=<1 0 0>}
      localJointOffset=<3 0 0>
     }]
   }
   n4 = Anim Dynamics - Bone: dyn_leaf_4(ComponentPose=@n3, Alpha=1, ExternalForce="0,0,0") {
    angularDampingOverride=1
    angularSpringConstant=20000
    bAngularSpring=true
    bOverrideAngularDamping=true
    bUseGravityOverride=true
    gravityScale=0.35
    simulationSpace=RootRelative
    alphaBoolBlend {blendOption=Linear}
    boundBone {boneName=dyn_leaf_4}
    componentLinearAccScale=<1 1 1>
    componentLinearVelScale=<1 1 1>
    gravityOverride {z=-980}
    physicsBodyDefinitions [{
      boundBone {boneName=dyn_leaf_4}
      boxExtents=<20 20 20>
      constraintSetup {angularLimitsMax=<10 45 45>, angularLimitsMin=<-10 -45 -45>, angularTarget=<1 0 0>}
      localJointOffset=<3 0 0>
     }]
   }
   n5 = Anim Dynamics - Bone: dyn_leaf_5(ComponentPose=@n4, Alpha=1, ExternalForce="0,0,0") {
    angularDampingOverride=1
    angularSpringConstant=20000
    bAngularSpring=true
    bOverrideAngularDamping=true
    bUseGravityOverride=true
    gravityScale=0.35
    simulationSpace=RootRelative
    alphaBoolBlend {blendOption=Linear}
    boundBone {boneName=dyn_leaf_5}
    componentLinearAccScale=<1 1 1>
    componentLinearVelScale=<1 1 1>
    gravityOverride {z=-980}
    physicsBodyDefinitions [{
      boundBone {boneName=dyn_leaf_5}
      boxExtents=<20 20 20>
      constraintSetup {angularLimitsMax=<10 45 45>, angularLimitsMin=<-10 -45 -45>, angularTarget=<1 0 0>}
      localJointOffset=<3 0 0>
     }]
   }
   n6 = Anim Dynamics - Bone: dyn_leaf_6(ComponentPose=@n5, Alpha=1, ExternalForce="0,0,0") {
    angularDampingOverride=1
    angularSpringConstant=20000
    bAngularSpring=true
    bOverrideAngularDamping=true
    bUseGravityOverride=true
    gravityScale=0.35
    simulationSpace=RootRelative
    alphaBoolBlend {blendOption=Linear}
    boundBone {boneName=dyn_leaf_6}
    componentLinearAccScale=<1 1 1>
    componentLinearVelScale=<1 1 1>
    gravityOverride {z=-980}
    physicsBodyDefinitions [{
      boundBone {boneName=dyn_leaf_6}
      boxExtents=<20 20 20>
      constraintSetup {angularLimitsMax=<10 45 45>, angularLimitsMin=<-10 -45 -45>, angularTarget=<1 0 0>}
      localJointOffset=<3 0 0>
     }]
   }
   n7 = RigidBody!(ComponentPose=@n6, Alpha, OverrideWorldGravity:Vector=Get Gravity Override(), ExternalForce) -> (Pose) {
    bClampLinearTranslationLimitToRefPose=true
    bOverrideWorldGravity=true
    overridePhysicsAsset=&/M_MED_BrakePedal_Pack_Physics
    simulationSpace=BaseBoneSpace
    alphaBoolBlend {blendOption=Linear}
    baseBoneRef {boneName=pack}
    componentAppliedLinearAccClamp=<3000 3000 3000>
    componentLinearAccScale=<0.3 0.3 0.3>
    componentLinearVelScale=<0.3 0.3 0.3>
    simSpaceSettings {
     maxAngularAcceleration=1000
     maxAngularVelocity=3
     maxLinearAcceleration=700
     maxLinearVelocity=500
     velocityScaleZ=0
     worldAlpha=0.225
     externalLinearDragV=<0.2 0.2>
    }
   }
   n8 = "Transform (Modify) Bone - Bone: pack"(
    ComponentPose=@n7
    Alpha=Get isMelee() -> (isMelee:double)
    Translation="0,0,0"
    Rotation=Get Backpack Rotate Pickaxe() -> (Backpack Rotate Pickaxe:Rotator)
   ) {
    rotationMode=BMM_Additive
    rotationSpace=BCS_BoneSpace
    translationMode=BMM_Additive
    translationSpace=BCS_BoneSpace
    alphaBoolBlend {blendOption=Linear}
    alphaScaleBiasClamp {interpSpeedDecreasing=5, interpSpeedIncreasing=5}
    boneToModify {boneName=pack}
    rotation {roll=20}
   }
   Output Pose(Result=Component To Local(ComponentPose=@n8))
   n9 = "Transform (Modify) Bone - Bone: pack"(ComponentPose=Local To Component(LocalPose=Input Pose()), Translation="0,0.75,0", Rotation="0,0,0") {
    translationMode=BMM_Additive
    translationSpace=BCS_BoneSpace
    alphaBoolBlend {blendOption=Linear}
    boneToModify {boneName=pack}
    translation {y=0.75}
   }
   n10 = "Transform (Modify) Bone - Bone: pack"(
    ComponentPose=@n9
    Translation=Get Backpack Translation() -> (Backpack Translation:Vector)
    Rotation=Get Backpack Rotate() -> (Backpack Rotate:Rotator)
   ) {
    rotationMode=BMM_Additive
    rotationSpace=BCS_BoneSpace
    translationMode=BMM_Additive
    translationSpace=BCS_BoneSpace
    alphaBoolBlend {blendOption=Linear}
    boneToModify {boneName=pack}
   }
   Save cached pose 'start_Cache'(Pose=Component To Local(ComponentPose=@n10)) {cacheName=start_Cache}
  ]}
 EventGraph {nodes [
   n1 = Event BlueprintUpdateAnimation()
   > n2 = Sequence()
   > n3 = Gravity Override State Function DEPRECATED(CharacterPartAnimInstance=Self-Reference(), GameTimeInSeconds=GetGameTimeInSeconds(), GravityOverrideParams=Get Gravity Override Params())
   > Set Gravity Override(Gravity Override=@n3)
   n4 = GetOwnerSkeletalMeshComponent()
   > n5 = Set head down rot(head down rot:double=GetAlphaFromReferenceTransformDelta(SkeletalMeshComponent=@n4, BoneName=spine_01, BaseName=head, InputRange=(X=0,Y=-45), TwistAxis="1,0,0")) -> (Output_Get:double)
   n6 = "To Float (Boolean)"(InBool=Get bIsPlayingMeleeAnim())
   n7 = float * float(A=@n1.DeltaTimeX, B=0.2) {is=title K2Node_PromotableOperator}
   n8 = Get isMelee()
   > Set isMelee(isMelee:double=float * float(A=@n5, B=FInterpTo(Current=@n8, Target=@n6, DeltaTime=@n7, InterpSpeed=5))) -> (Output_Get:double)
   n9 = BreakRotator(InRot=BreakTransform(InTransform=GetDeltaTransformFromRefPose(self=@n4, BoneName=upperarm_l, BaseName=clavicle_l)).Rotation)
   > Set Backpack Rotate Pickaxe(Backpack Rotate Pickaxe:Rotator=MakeRotator(Roll=MapRangeClamped(Value=@n9.Yaw, InRangeA=0, InRangeB=90, OutRangeB=30))) -> (Output_Get:Rotator)
   n10 = GetOwnerSkeletalMeshComponent()
   n11 = BreakRotator(InRot=BreakTransform(InTransform=GetDeltaTransformFromRefPose(self=@n10, BoneName=spine_05, BaseName=spine_01)).Rotation)
   Set Backpack Rotate(execute=@n2.then_1, Backpack Rotate:Rotator=MakeRotator(Roll=MapRangeClamped(Value=@n11.Yaw, InRangeA=0, InRangeB=-40, OutRangeB=30))) -> (Output_Get:Rotator)
   > Set Backpack Translation(Backpack Translation:Vector=MakeVector(Y=MapRangeClamped(Value=@n11.Yaw, InRangeA=0, InRangeB=-30, OutRangeB=-5))) -> (Output_Get:Vector)
   TryGetPawnOwner()
  ]}
]
variables [
 Backpack Rotate {type=Rotator}
 Backpack Rotate Pickaxe {type=Rotator}
 Backpack Translation {type=Vector}
 Gravity Override {type=Vector}
 Gravity Override Params {type=GravityOverrideParamsStruct}
 head down rot {type=number}
 isMelee {type=number}
]
